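# Physics asset for the TreyCozy hair of the Pasifika medium female player character's face accessory mesh, defining how the hair swings.
# Boxes cover the main hair bones and small capsules the sub-strands, with larger capsules for the head and upper spine.
# Right, back and branching left chains hang from the head through narrow swing-twist limits, looser on some sub-strands.
# The head-to-spine_05 joint is free in swing and twist.
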PhysicsAsset /Game/Characters/Player/Female/Medium/Bodies/F_MED_Prime_PASIFIKA/Meshes/Parts/F_MED_Prime_PASIFIKA_FaceAcc_hair_PhysicsAsset_TreyCozy
bodies {
 dyn_back_hair_01 [{shapeType=Box, extentX=5, extentY=2.25, extentZ=7, shapeName=dyn_back_hair_01_box, center=<0.25 0 0>}]
 dyn_back_hair_02 [{shapeType=Box, extentX=5, extentY=3, extentZ=7, shapeName=dyn_back_hair_02_box, center=<0.25 0 0>}]
 dyn_back_hair_s1 [{shapeType=Capsule, length=2, radius=1, shapeName=dyn_back_hair_s1_capsule, center=<2.25 0 0>, rotation=<rot 90 0 0>}]
 dyn_left_hair_01 [{shapeType=Box, extentX=5, extentY=7, extentZ=1, shapeName=dyn_left_hair_01_box, center=<0.25 0 0>}]
 dyn_left_hair_02 [{shapeType=Box, extentX=5, extentY=7, extentZ=2, shapeName=dyn_left_hair_02_box, center=<0.25 0 0>}]
 dyn_left_hair_b [{shapeType=Box, extentX=5, extentY=7, extentZ=2.25, shapeName=dyn_left_hair_b_box, center=<0.25 0 0>}]
 dyn_left_hair_bs1 [{shapeType=Capsule, length=2, radius=1, shapeName=dyn_left_hair_bs1_capsule, center=<2.25 0 0>, rotation=<rot 90 0 0>}]
 dyn_left_hair_f [{shapeType=Box, extentX=5, extentY=7, extentZ=2.25, shapeName=dyn_left_hair_f_box, center=<-0.75 0 0>}]
 dyn_left_hair_fs1 [{shapeType=Capsule, length=2, radius=1, shapeName=dyn_left_hair_fs1_capsule, center=<2.25 0 0>, rotation=<rot 90 0 0>}]
 dyn_left_hair_fs2 [{shapeType=Capsule, length=2, radius=1, shapeName=dyn_left_hair_fs2_capsule, center=<1.5 0 0>, rotation=<rot 90 0 0>}]
 dyn_left_hair_m [{shapeType=Box, extentX=5, extentY=7, extentZ=1, shapeName=dyn_left_hair_m_box, center=<0.25 0 0>, rotation=<rot 0 0 45>}]
 dyn_right_hair_01 [{shapeType=Box, extentX=5, extentY=7, extentZ=2.75, shapeName=dyn_right_hair_01_box, center=<0.5 0 0>}]
 dyn_right_hair_s1 [{shapeType=Capsule, length=2, radius=1, shapeName=dyn_right_hair_s1_capsule, center=<2.25 0 0>, rotation=<rot 90 0 0>}]
 dyn_right_hair_s2 [{shapeType=Capsule, length=2, radius=1, shapeName=dyn_right_hair_s2_capsule, center=<2.25 0 0>, rotation=<rot 90 0 0>}]
 head [{shapeType=Capsule, length=8.0176, radius=6.4787, center=<0 -0.11 0>, rotation=<rot 90 0 0>}]
 spine_04 [
  {shapeType=Capsule, length=18.5, radius=8.375, shapeName=spine_05_capsule, center=<3.3658 0 -1>, rotation=<rot 90 0 0>}
  {shapeType=Capsule, length=26.25, radius=4.5, shapeName=spine_05_capsule, center=<21.5817 -3.069 0>}
 ]
 spine_05 [
  {shapeType=Capsule, length=17.5, radius=9.375, shapeName=spine_05_capsule, center=<3.3658 0 -1>, rotation=<rot 90 0 0>}
  {shapeType=Capsule, length=26.25, radius=4.5, shapeName=spine_05_capsule, center=<6.4953 -3.1293 0>}
 ]
}
constraints [
 {bone1Name=dyn_right_hair_01, bone2Name=head, swing1LimitDegrees=20, swing2LimitDegrees=15, twistLimitDegrees=2}
 {bone1Name=dyn_right_hair_s1, bone2Name=dyn_right_hair_01, swing1LimitDegrees=45, swing2LimitDegrees=45, twistLimitDegrees=2}
 {bone1Name=dyn_right_hair_s2, bone2Name=dyn_right_hair_01, swing1LimitDegrees=45, swing2LimitDegrees=45, twistLimitDegrees=2}
 {bone1Name=dyn_left_hair_01, bone2Name=head, swing1LimitDegrees=12, swing2LimitDegrees=15, twistLimitDegrees=2}
 {bone1Name=dyn_left_hair_02, bone2Name=dyn_left_hair_01, swing1LimitDegrees=12, swing2LimitDegrees=15, twistLimitDegrees=2}
 {bone1Name=dyn_left_hair_f, bone2Name=dyn_left_hair_02, swing1LimitDegrees=15, swing2LimitDegrees=15, twistLimitDegrees=2}
 {bone1Name=dyn_left_hair_fs1, bone2Name=dyn_left_hair_f, swing1LimitDegrees=15, swing2LimitDegrees=15, twistLimitDegrees=2}
 {bone1Name=dyn_left_hair_fs2, bone2Name=dyn_left_hair_f, swing1LimitDegrees=15, swing2LimitDegrees=15, twistLimitDegrees=2}
 {bone1Name=dyn_left_hair_m, bone2Name=dyn_left_hair_02, swing1LimitDegrees=15, swing2LimitDegrees=15, twistLimitDegrees=2}
 {bone1Name=dyn_left_hair_b, bone2Name=dyn_left_hair_02, swing1LimitDegrees=15, swing2LimitDegrees=15, twistLimitDegrees=2}
 {bone1Name=dyn_left_hair_bs1, bone2Name=dyn_left_hair_b, swing1LimitDegrees=45, swing2LimitDegrees=45, twistLimitDegrees=2}
 {bone1Name=dyn_back_hair_01, bone2Name=head, swing1LimitDegrees=10, swing2LimitDegrees=5, twistLimitDegrees=2}
 {bone1Name=dyn_back_hair_02, bone2Name=dyn_back_hair_01, swing1LimitDegrees=10, swing2LimitDegrees=5, twistLimitDegrees=2}
 {
  bone1Name=head
  bone2Name=spine_05
  swing1LimitDegrees=45
  swing1Motion=Free
  swing2LimitDegrees=45
  swing2Motion=Free
  twistLimitDegrees=45
  twistMotion=Free
 }
 {bone1Name=dyn_back_hair_s1, bone2Name=dyn_back_hair_02, swing1LimitDegrees=10, swing2LimitDegrees=20, twistLimitDegrees=2}
]
