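# Animation Blueprint for the TCozy variant of the female PASIFIKA hair accessory skeleton, adding physics and corrective posing to the hair.
# It copies the attached parent's pose, simulates RigidBody in head space with the TreyCozy physics asset, then applies additive offsets to back and right hair bones.
# Each update maps head-versus-spine_05 yaw to Neck Alpha, which weights all but one constant offset.
# Neck Alpha 02 feeds only disconnected leftover nodes, alongside an unused second RigidBody.
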AnimBlueprint /BRCosmetics/Characters/Player/Female/Medium/Bodies/F_MED_Prime_PASIFIKA/Meshes/Parts/F_MED_Prime_PASIFIKA_FaceAcc_hair_Skeleton_TCozy_AnimBlueprint
target_skeleton=/Game/Characters/Player/Female/Medium/Bodies/F_MED_Prime_PASIFIKA/Meshes/Parts/F_MED_Prime_PASIFIKA_FaceAcc_hair_Skeleton
graphs [
 AnimGraph {nodes [
   "Transform (Modify) Bone - Bone: dyn_back_hair_01"(Alpha=Get Neck Alpha() -> (Neck Alpha:double), Translation="-2,-1.75,0", Rotation="0,0,0", Scale="1,1,1") {
    translationMode=BMM_Additive
    translationSpace=BCS_BoneSpace
    alphaBoolBlend {blendOption=Linear}
    boneToModify {boneName=dyn_back_hair_01}
    translation=<-2 -1.75>
   }
   n1 = "Transform (Modify) Bone - Bone: dyn_back_hair_01"(Alpha=Get Neck Alpha(), Translation="-3,-1.75,0", Rotation="10,-0.000,10.000", Scale="1,1,1") {
    rotationMode=BMM_Additive
    translationMode=BMM_Additive
    translationSpace=BCS_BoneSpace
    alphaBoolBlend {blendOption=Linear}
    boneToModify {boneName=dyn_back_hair_01}
    rotation {pitch=10, roll=10}
    translation=<-3 -1.75>
   }
   "Transform (Modify) Bone - Bone: dyn_right_hair_01"(ComponentPose=@n1, Alpha=Get Neck Alpha 02() -> (Neck Alpha 02:double), Translation="-3,-2,2", Rotation="-15,-30,5", Scale="1,1,1") {
    rotationMode=BMM_Additive
    rotationSpace=BCS_BoneSpace
    translationMode=BMM_Additive
    translationSpace=BCS_BoneSpace
    alphaBoolBlend {blendOption=Linear}
    boneToModify {boneName=dyn_right_hair_01}
    rotation=<rot -15 -30 5>
    translation=<-3 -2 2>
   }
   n2 = "Transform (Modify) Bone - Bone: dyn_back_hair_01"(Alpha=Get Neck Alpha(), Translation="-4,-2.5,0", Rotation="0.000,0,5.000", Scale="1,1,1") {
    rotationMode=BMM_Additive
    translationMode=BMM_Additive
    translationSpace=BCS_BoneSpace
    alphaBoolBlend {blendOption=Linear}
    boneToModify {boneName=dyn_back_hair_01}
    rotation {roll=5}
    translation=<-4 -2.5>
   }
   n3 = Get Neck Alpha 02()
   n4 = "Transform (Modify) Bone - Bone: dyn_right_hair_01"(ComponentPose=@n2, Alpha=@n3, Translation="-3,-2,2", Rotation="-5,-10,5", Scale="1,1,1") {
    rotationMode=BMM_Additive
    rotationSpace=BCS_BoneSpace
    translationMode=BMM_Additive
    translationSpace=BCS_BoneSpace
    alphaBoolBlend {blendOption=Linear}
    boneToModify {boneName=dyn_right_hair_01}
    rotation=<rot -5 -10 5>
    translation=<-3 -2 2>
   }
   "Transform (Modify) Bone - Bone: dyn_back_hair_02"(ComponentPose=@n4, Alpha=@n3, Translation="-0.000,0.000,2.205", Rotation="0,0,0", Scale="1,1,1") {
    rotationMode=BMM_Additive
    translationMode=BMM_Additive
    translationSpace=BCS_WorldSpace
    alphaBoolBlend {blendOption=Linear}
    boneToModify {boneName=dyn_back_hair_02}
    translation {z=2.205}
   }
   RigidBody() {
    bClampLinearTranslationLimitToRefPose=true
    bOverrideWorldGravity=true
    overridePhysicsAsset=/Game/Characters/Player/Female/Medium/Bodies/F_MED_Prime_PASIFIKA/Meshes/Parts/F_MED_Prime_PASIFIKA_FaceAcc_hair_PhysicsAsset_2
    simulationSpace=BaseBoneSpace
    alphaBoolBlend {blendOption=Linear}
    baseBoneRef {boneName=head}
    componentAppliedLinearAccClamp=<3000 3000 3000>
    componentLinearAccScale=<0.3 0.3 0.35>
    componentLinearVelScale=<0.3 0.3 0.3>
    overrideWorldGravity {z=-800}
   }
   n5 = RigidBody(ComponentPose=Local To Component(LocalPose=Copy Pose From Mesh() {bCopyCurves=true, bUseAttachedParent=true})) {
    bClampLinearTranslationLimitToRefPose=true
    bOverrideWorldGravity=true
    overridePhysicsAsset=/Game/Characters/Player/Female/Medium/Bodies/F_MED_Prime_PASIFIKA/Meshes/Parts/F_MED_Prime_PASIFIKA_FaceAcc_hair_PhysicsAsset_TreyCozy
    simulationSpace=BaseBoneSpace
    alphaBoolBlend {blendOption=Linear}
    baseBoneRef {boneName=head}
    componentAppliedLinearAccClamp=<3000 3000 3000>
    componentLinearAccScale=<0.31 0.31 0.36>
    componentLinearVelScale=<0.31 0.31 0.31>
    overrideWorldGravity {z=-800}
   }
   n6 = Get Neck Alpha()
   n7 = "Transform (Modify) Bone - Bone: dyn_back_hair_01"(ComponentPose=@n5, Alpha=@n6, Translation="0,2,0", Rotation="2,0,0", Scale="1,1,1") {
    rotationMode=BMM_Additive
    rotationSpace=BCS_BoneSpace
    translationMode=BMM_Additive
    translationSpace=BCS_BoneSpace
    alphaBoolBlend {blendOption=Linear}
    boneToModify {boneName=dyn_back_hair_01}
    rotation {pitch=2}
    translation {y=2}
   }
   n8 = "Transform (Modify) Bone - Bone: dyn_back_hair_02"(ComponentPose=@n7, Alpha=1, Translation="0,0,0", Rotation="0,0,-10", Scale="1,1,1") {
    rotationMode=BMM_Additive
    scaleMode=BMM_Additive
    translationMode=BMM_Additive
    translationSpace=BCS_WorldSpace
    alphaBoolBlend {blendOption=Linear}
    boneToModify {boneName=dyn_back_hair_02}
    rotation {roll=-10}
   }
   n9 = "Transform (Modify) Bone - Bone: dyn_back_hair_02"(ComponentPose=@n8, Alpha=@n6, Translation="0,0,3", Rotation="0,0,-10", Scale="0.9,1,1") {
    rotationMode=BMM_Additive
    scaleMode=BMM_Additive
    translationMode=BMM_Additive
    translationSpace=BCS_WorldSpace
    alphaBoolBlend {blendOption=Linear}
    boneToModify {boneName=dyn_back_hair_02}
    rotation {roll=-10}
    scale {x=0.9}
    translation {z=3}
   }
   n10 = "Transform (Modify) Bone - Bone: dyn_right_hair_01"(ComponentPose=@n9, Alpha=@n6, Translation="0,0,1", Rotation="0,0,0", Scale="0.9,1,1") {
    rotationMode=BMM_Additive
    scaleMode=BMM_Additive
    translationMode=BMM_Additive
    translationSpace=BCS_WorldSpace
    alphaBoolBlend {blendOption=Linear}
    boneToModify {boneName=dyn_right_hair_01}
    scale {x=0.9}
    translation {z=1}
   }
   n11 = "Transform (Modify) Bone - Bone: dyn_back_hair_s1"(ComponentPose=@n10, Alpha=@n6, Translation="-0.000,0.000,1", Rotation="5,0.000,0.000", Scale="1,1,1") {
    rotationMode=BMM_Additive
    translationMode=BMM_Additive
    translationSpace=BCS_WorldSpace
    alphaBoolBlend {blendOption=Linear}
    boneToModify {boneName=dyn_back_hair_s1}
    rotation {pitch=5}
    translation {z=1}
   }
   n12 = "Transform (Modify) Bone - Bone: dyn_back_hair_s2"(ComponentPose=@n11, Alpha=@n6, Translation="-0.000,0.000,1", Rotation="5,0.000,-30", Scale="1,1,1") {
    rotationMode=BMM_Additive
    translationMode=BMM_Additive
    translationSpace=BCS_WorldSpace
    alphaBoolBlend {blendOption=Linear}
    boneToModify {boneName=dyn_back_hair_s2}
    rotation {pitch=5, roll=-30}
    translation {z=1}
   }
   n13 = "Transform (Modify) Bone - Bone: dyn_back_hair_s3"(ComponentPose=@n12, Alpha=@n6, Translation="-0.000,0.000,4", Rotation="5,0.000,-20", Scale="1,1,1") {
    rotationMode=BMM_Additive
    translationMode=BMM_Additive
    translationSpace=BCS_WorldSpace
    alphaBoolBlend {blendOption=Linear}
    boneToModify {boneName=dyn_back_hair_s3}
    rotation {pitch=5, roll=-20}
    translation {z=4}
   }
   Output Pose(Result=Component To Local(ComponentPose=@n13)) {layerGroup=None}
   Get Neck Alpha()
  ]}
 EventGraph {nodes [
   Event BlueprintUpdateAnimation()
   n1 = GetOwnerSkeletalMeshComponent()
   n2 = BreakRotator(InRot=BreakTransform(InTransform=GetDeltaTransformFromRefPose(self=@n1, BoneName=head, BaseName=spine_05)).Rotation)
   > Set Neck Alpha(Neck Alpha:double=MapRangeClamped(Value=@n2.Yaw, InRangeA=-15, InRangeB=-40)) -> (Output_Get:double)
   > Set Neck Alpha 02(Neck Alpha 02:double=MapRangeClamped(Value=@n2, InRangeA=0, InRangeB=40)) -> (Output_Get:double)
   TryGetPawnOwner()
  ]}
]
variables [
 Neck Alpha {type=number}
 Neck Alpha 02 {type=number}
]
